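; Dark Muster faction chest Blueprint that supplies the opening animation and effects for a lootable container.
; The opening event plays a one-shot timeline that swings the left and right arm lids in opposite directions.
; The same timeline feeds its time into the mesh materials' CODE_Timestamp parameter to sync the shader with the motion.
; The VFX event spawns a Knight chest burst Niagara system at the mesh and fades out the ambient sound.
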(event EventPlayChestOpeningTimeline
  (Collision|IgnoreActorwhenMoving (Variables|Default|GetL_ArmLid) self true)
  (Utilities|FlowControl|DoOnce false
    (:Completed
      (bind (_direction _time _openpitch) (Timeline|Timeline_Open Play)
        (:Update
          (Rendering|Material|SetScalarParameterValueonMaterials (Variables|Debug|GetStaticMeshComponent) "CODE_Timestamp" _time)
          (Transformation|SetRelativeRotation (Variables|Default|GetL_ArmLid) (Math|Rotator|MakeRotator 0.0 (Math|Float|Lerp 0.0 -155.0 _openpitch)) false false)
          (Transformation|SetRelativeRotation (Variables|Default|GetR_ArmLid) (Math|Rotator|MakeRotator 0.0 (Math|Float|Lerp 0.0 155.0 _openpitch)) false false))))))

(event EventSpawnChestOpenVFX
  (Niagara|SpawnSystematLocation "/Game/Effects/Fort_Effects/World/NewChests/NS_ChestBurst_Knight.NS_ChestBurst_Knight" (.location (Transformation|GetWorldTransform (Variables|Debug|GetStaticMeshComponent))) (.rotation (Transformation|GetWorldTransform (Variables|Debug|GetStaticMeshComponent))))
  (Audio|Components|Audio|FadeOut (Variables|Default|GetChest_Ambient_Sound) 0.15 0.0 "Linear" :!class /Script/Engine.AudioComponent))

(fn ConstructionScript
  (|CallParentFunction))
